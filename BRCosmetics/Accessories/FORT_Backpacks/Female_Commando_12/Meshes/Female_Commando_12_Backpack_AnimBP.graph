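# Anim Blueprint for the Female_Commando_12 backpack cosmetic that adds procedural secondary motion to its rings, volume bone and bullets.
# The input pose runs through a serial chain of Anim Dynamics nodes in component space, simulating in actor space.
# Both rings and the eight bullets use angular springs with per-bone limits, while dyn_volume uses a damped linear spring with rotation locked.
# The event graph holds only unconnected stub nodes, with no logic or variables.
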AnimBlueprint /BRCosmetics/Accessories/FORT_Backpacks/Female_Commando_12/Meshes/Female_Commando_12_Backpack_AnimBP
target_skeleton=&/Female_Commando_12_Skeleton
graphs [
 AnimGraph {nodes [
   AnimGraphNode_SubInput_0 = Input Pose()
   n1 = Anim Dynamics - Bone: dyn_ring_l(ComponentPose=Local To Component(LocalPose=@AnimGraphNode_SubInput_0), Alpha=1, ExternalForce="0,0,0") {
    alpha=0.9
    angularBiasOverride=0.9
    angularSpringConstant=1000
    bAngularSpring=true
    bGravityOverrideInSimSpace=true
    bOverrideAngularBias=true
    numSolverIterationsPostUpdate=4
    numSolverIterationsPreUpdate=16
    simulationSpace=Actor
    alphaBoolBlend {blendOption=Linear}
    boundBone {boneName=dyn_ring_l}
    physicsBodyDefinitions [{
      sphereCollisionRadius=0
      boundBone {boneName=dyn_ring_l}
      boxExtents=<5 5 5>
      constraintSetup {angularTargetAxis=AxisZ, angularLimitsMax=<10 30 20>, angularLimitsMin=<-10 -30 -20>, angularTarget=<0 0 1>}
      localJointOffset=<0 0 0>
     }]
   }
   n2 = Anim Dynamics - Bone: dyn_ring_r(ComponentPose=@n1, Alpha=1, ExternalForce="0,0,0") {
    alpha=0.9
    angularBiasOverride=0.9
    angularSpringConstant=1000
    bAngularSpring=true
    bGravityOverrideInSimSpace=true
    bOverrideAngularBias=true
    numSolverIterationsPostUpdate=4
    numSolverIterationsPreUpdate=16
    simulationSpace=Actor
    alphaBoolBlend {blendOption=Linear}
    boundBone {boneName=dyn_ring_r}
    physicsBodyDefinitions [{
      sphereCollisionRadius=0
      boundBone {boneName=dyn_ring_r}
      boxExtents=<5 5 5>
      constraintSetup {angularTargetAxis=AxisZ, angularLimitsMax=<10 30 20>, angularLimitsMin=<-10 -30 -20>, angularTarget=<0 0 1>}
      localJointOffset=<0 0 0>
     }]
   }
   n3 = Anim Dynamics - Bone: dyn_volume(ComponentPose=@n2, Alpha=1, ExternalForce="0,0,0") {
    alpha=0.9
    bGravityOverrideInSimSpace=true
    bLinearSpring=true
    bOverrideLinearDamping=true
    linearDampingOverride=0.9
    linearSpringConstant=300
    numSolverIterationsPostUpdate=4
    numSolverIterationsPreUpdate=16
    simulationSpace=Actor
    alphaBoolBlend {blendOption=Linear}
    boundBone {boneName=dyn_volume}
    physicsBodyDefinitions [{
      sphereCollisionRadius=0
      boundBone {boneName=dyn_volume}
      boxExtents=<5 5 5>
      constraintSetup {
       angularTargetAxis=AxisZ
       angularLimitsMax=<0 0 0>
       angularLimitsMin=<0 0 0>
       angularTarget=<0 0 1>
       linearAxesMax=<2 4 2>
       linearAxesMin=<-5 -4 -2>
      }
      localJointOffset=<0 0 0>
     }]
   }
   n4 = Anim Dynamics - Bone: dyn_bullet_01(ComponentPose=@n3, Alpha=1, ExternalForce="0,0,0") {
    alpha=0.9
    angularBiasOverride=0.9
    angularSpringConstant=500
    bAngularSpring=true
    bGravityOverrideInSimSpace=true
    bOverrideAngularBias=true
    numSolverIterationsPostUpdate=4
    numSolverIterationsPreUpdate=16
    simulationSpace=Actor
    alphaBoolBlend {blendOption=Linear}
    boundBone {boneName=dyn_bullet_01}
    physicsBodyDefinitions [{
      sphereCollisionRadius=0
      boundBone {boneName=dyn_bullet_01}
      boxExtents=<5 10 5>
      constraintSetup {angularTargetAxis=AxisZ, angularLimitsMax=<5 10 40>, angularLimitsMin=<-5 -50 -40>, angularTarget=<0 0 1>}
      localJointOffset=<0.3 0 0>
     }]
   }
   n5 = Anim Dynamics - Bone: dyn_bullet_02(ComponentPose=@n4, Alpha=1, ExternalForce="0,0,0") {
    alpha=0.9
    angularBiasOverride=0.9
    angularSpringConstant=500
    bAngularSpring=true
    bGravityOverrideInSimSpace=true
    bOverrideAngularBias=true
    numSolverIterationsPostUpdate=4
    numSolverIterationsPreUpdate=16
    simulationSpace=Actor
    alphaBoolBlend {blendOption=Linear}
    boundBone {boneName=dyn_bullet_02}
    physicsBodyDefinitions [{
      sphereCollisionRadius=0
      boundBone {boneName=dyn_bullet_02}
      boxExtents=<5 10 5>
      constraintSetup {angularTargetAxis=AxisZ, angularLimitsMax=<5 10 30>, angularLimitsMin=<-5 -50 -30>, angularTarget=<0 0 1>}
      localJointOffset=<0.3 0 0>
     }]
   }
   n6 = Anim Dynamics - Bone: dyn_bullet_03(ComponentPose=@n5, Alpha=1, ExternalForce="0,0,0") {
    alpha=0.9
    angularBiasOverride=0.9
    angularSpringConstant=500
    bAngularSpring=true
    bGravityOverrideInSimSpace=true
    bOverrideAngularBias=true
    numSolverIterationsPostUpdate=4
    numSolverIterationsPreUpdate=16
    simulationSpace=Actor
    alphaBoolBlend {blendOption=Linear}
    boundBone {boneName=dyn_bullet_03}
    physicsBodyDefinitions [{
      sphereCollisionRadius=0
      boundBone {boneName=dyn_bullet_03}
      boxExtents=<5 10 5>
      constraintSetup {angularTargetAxis=AxisZ, angularLimitsMax=<5 5 20>, angularLimitsMin=<-5 -50 -20>, angularTarget=<0 0 1>}
      localJointOffset=<0.3 0 0>
     }]
   }
   n7 = Anim Dynamics - Bone: dyn_bullet_04(ComponentPose=@n6, Alpha=1, ExternalForce="0,0,0") {
    alpha=0.9
    angularBiasOverride=0.9
    angularSpringConstant=500
    bAngularSpring=true
    bGravityOverrideInSimSpace=true
    bOverrideAngularBias=true
    numSolverIterationsPostUpdate=4
    numSolverIterationsPreUpdate=16
    simulationSpace=Actor
    alphaBoolBlend {blendOption=Linear}
    boundBone {boneName=dyn_bullet_04}
    physicsBodyDefinitions [{
      sphereCollisionRadius=0
      boundBone {boneName=dyn_bullet_04}
      boxExtents=<5 10 5>
      constraintSetup {angularTargetAxis=AxisZ, angularLimitsMax=<5 5 20>, angularLimitsMin=<-5 -50 -20>, angularTarget=<0 0 1>}
      localJointOffset=<0.3 0 0>
     }]
   }
   n8 = Anim Dynamics - Bone: dyn_bullet_05(ComponentPose=@n7, Alpha=1, ExternalForce="0,0,0") {
    alpha=0.9
    angularBiasOverride=0.9
    angularSpringConstant=500
    bAngularSpring=true
    bGravityOverrideInSimSpace=true
    bOverrideAngularBias=true
    numSolverIterationsPostUpdate=4
    numSolverIterationsPreUpdate=16
    simulationSpace=Actor
    alphaBoolBlend {blendOption=Linear}
    boundBone {boneName=dyn_bullet_05}
    physicsBodyDefinitions [{
      sphereCollisionRadius=0
      boundBone {boneName=dyn_bullet_05}
      boxExtents=<5 10 5>
      constraintSetup {angularTargetAxis=AxisZ, angularLimitsMax=<5 5 20>, angularLimitsMin=<-5 -50 -20>, angularTarget=<0 0 1>}
      localJointOffset=<0.3 0 0>
     }]
   }
   n9 = Anim Dynamics - Bone: dyn_bullet_06(ComponentPose=@n8, Alpha=1, ExternalForce="0,0,0") {
    alpha=0.9
    angularBiasOverride=0.9
    angularSpringConstant=500
    bAngularSpring=true
    bGravityOverrideInSimSpace=true
    bOverrideAngularBias=true
    numSolverIterationsPostUpdate=4
    numSolverIterationsPreUpdate=16
    simulationSpace=Actor
    alphaBoolBlend {blendOption=Linear}
    boundBone {boneName=dyn_bullet_06}
    physicsBodyDefinitions [{
      sphereCollisionRadius=0
      boundBone {boneName=dyn_bullet_06}
      boxExtents=<5 10 5>
      constraintSetup {angularTargetAxis=AxisZ, angularLimitsMax=<5 10 20>, angularLimitsMin=<-5 -50 -20>, angularTarget=<0 0 1>}
      localJointOffset=<0.3 0 0>
     }]
   }
   n10 = Anim Dynamics - Bone: dyn_bullet_07(ComponentPose=@n9, Alpha=1, ExternalForce="0,0,0") {
    alpha=0.9
    angularBiasOverride=0.9
    angularSpringConstant=500
    bAngularSpring=true
    bGravityOverrideInSimSpace=true
    bOverrideAngularBias=true
    numSolverIterationsPostUpdate=4
    numSolverIterationsPreUpdate=16
    simulationSpace=Actor
    alphaBoolBlend {blendOption=Linear}
    boundBone {boneName=dyn_bullet_07}
    physicsBodyDefinitions [{
      sphereCollisionRadius=0
      boundBone {boneName=dyn_bullet_07}
      boxExtents=<5 10 5>
      constraintSetup {angularTargetAxis=AxisZ, angularLimitsMax=<5 5 30>, angularLimitsMin=<-5 -50 -30>, angularTarget=<0 0 1>}
      localJointOffset=<0.3 0 0>
     }]
   }
   n11 = Anim Dynamics - Bone: dyn_bullet_08(ComponentPose=@n10, Alpha=1, ExternalForce="0,0,0") {
    alpha=0.9
    angularBiasOverride=0.9
    angularSpringConstant=500
    bAngularSpring=true
    bGravityOverrideInSimSpace=true
    bOverrideAngularBias=true
    numSolverIterationsPostUpdate=4
    numSolverIterationsPreUpdate=16
    simulationSpace=Actor
    alphaBoolBlend {blendOption=Linear}
    boundBone {boneName=dyn_bullet_08}
    physicsBodyDefinitions [{
      sphereCollisionRadius=0
      boundBone {boneName=dyn_bullet_08}
      boxExtents=<5 10 5>
      constraintSetup {angularTargetAxis=AxisZ, angularLimitsMax=<5 0 40>, angularLimitsMin=<-5 -50 -40>, angularTarget=<0 0 1>}
      localJointOffset=<0.3 0 0>
     }]
   }
   Output Pose(Result=Component To Local(ComponentPose=@n11)) {layerGroup=None}
  ]}
 EventGraph {nodes [
   TryGetPawnOwner()
   Event BlueprintUpdateAnimation()
  ]}
]
variables []
