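# Physics asset for the Female_Commando_12 backpack cosmetic mesh in Battle Royale cosmetics.
# It defines a single capsule body on the backpack_root bone, offset and rotated from the bone.
# No constraints exist, so there are no jointed or simulated secondary parts.
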PhysicsAsset /BRCosmetics/Accessories/FORT_Backpacks/Female_Commando_12/Meshes/Female_Commando_12_PhysicsAsset
bodies {backpack_root [{shapeType=Capsule, length=20.6508, radius=17.2711, center=<-9.0196 -10.243 -0.3089>, rotation=<rot 15.529 -90 -79.9488>}]}
constraints []
